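# M_EngineLogo is a translucent, two-sided material in the VR Editor device content, built around the Unreal Engine logo texture.
# It contains a texture sample of the UE_Logo texture from the Vive device folder and a Color vector parameter defaulting to white with zero alpha.
# It uses Default Lit shading and enables responsive anti-aliasing.
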Material /Engine/VREditor/Devices/M_EngineLogo
graph {
 properties {bEnableResponsiveAA=true, blendMode=BLEND_Translucent, pixelDepthOffsetMode=PDOM_Legacy, twoSided=true, usedShadingModels=Default Lit}
 nodes [
  TextureSample() {texture=/Engine/VREditor/Devices/Vive/UE_Logo}
  VectorParameter() {parameterName=Color, defaultValue=<rgba 1 1 1 0>}
 ]
}
